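# Base masked, lit material for rock chunk particle sprites, using 2x2 chunk textures.
# Particle speed over a Speed parameter, clamped, blends sharp normal and mask textures toward blurred versions, faking motion blur.
# Normals are transformed out of tangent space, particle alpha scales the mask, and particle color multiplies a ShadowColor tint.
# Feature level switches give ES2 and ES3.1 only the sharp textures, without the particle alpha scaling.
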Material /Figment_S05/Effects/Fort_Effects/Materials/Rocks/M_Chunks_Base
graph {
 properties {
  bEnableResponsiveAA=true
  bTangentSpaceNormal=false
  bUsedWithParticleSprites=true
  blendMode=BLEND_Masked
  pixelDepthOffsetMode=PDOM_Legacy
  usedShadingModels=Default Lit
 }
 nodes [
  n1 = Transform(None=TextureSampleParameter2D() {parameterName=Normal, samplerType=SAMPLERTYPE_Normal, texture=/Game/Effects/Fort_Effects/Textures/Rocks/T_Chunks_2x2_N}.RGB)
  n2 = Clamp(None=Divide(A=ParticleSpeed(), B=ScalarParameter() {defaultValue=60, parameterName=Speed}))
  n3 = LinearInterpolate(
   A=@n1
   B=Transform(None=TextureSampleParameter2D() {parameterName=Normal, samplerType=SAMPLERTYPE_Normal, texture=/Game/Effects/Fort_Effects/Textures/Rocks/T_ChunksBlur_2x2_N}.RGB)
   Alpha=@n2
  )
  FeatureLevelSwitch(Default=@n3, ES2=@n1, ES3_1=@n1, SM4_REMOVED=@n3)
  n4 = ParticleColor()
  n5 = TextureSampleParameter2D() {parameterName=Color and Alpha, texture=/Game/Effects/Fort_Effects/Textures/Rocks/T_Chunks_2x2_D}
  n6 = Multiply(
   A=@n4.A
   B=LinearInterpolate(
    A=@n5.G
    B=TextureSampleParameter2D() {parameterName=Color and Alpha, texture=/Game/Effects/Fort_Effects/Textures/Rocks/T_ChunksBlur_2x2_D}.G
    Alpha=@n2
   )
  )
  FeatureLevelSwitch(Default=@n6, ES2=@n5.G, ES3_1=@n5.G, SM4_REMOVED=@n6)
  Multiply(A=VectorParameter() {parameterName=ShadowColor, defaultValue=<rgba 0.1 0.1 0.1 1>}.RGB, B=@n4.RGB)
 ]
}
